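# Master character material for the SliceVine outfit, layering decal patterns onto the 2023 character parent's PreFX and PostFX functions.
# FN hair shading is blended in by hair mask before the decals.
# An Is Cherry Variant switch selects between masked, recolorable patterns with adjustable metallic, roughness and normal flattening, or decal RGB blended into base color.
# Tech Art Mask channels control decal placement and the cherry variant's alternate color.
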Material /BRCosmetics/Characters/Player/Male/Medium/Bodies/M_MED_SliceVine_LoopPlum/Skins/Loop/Materials/M_SliceVine_Master
graph {
 properties {
  bUseMaterialAttributes=true
  bUsedWithInstancedStaticMeshes=true
  bUsedWithMeshDeformer=true
  bUsedWithMorphTargets=true
  bUsedWithSkeletalMesh=true
  pixelDepthOffsetMode=PDOM_Legacy
  shadingModel=MSM_FromMaterialExpression
  usedShadingModels=Default Lit | Subsurface | Clear Coat | Subsurface Profile | Cloth
 }
 nodes [
  Constant3Vector() {constant=<rgba 0.216 0.216 1 1>}
  Constant() {r=1}
  n1 = VectorParameter() {group=01_SliceVine, parameterName=Color, sortPriority=10, defaultValue=<rgba 0.5647 0.5583 0.0033 1>}
  n2 = MaterialFunctionCall() {materialFunction=&MF_CharacterParent_2023_PreFX}
  n3 = MaterialFunctionCall!(In=@n2.Result, Mask Input=@n2.M) {materialFunction=/Game/Global/Materials/Character_Parent_2023/Hair_Parent_2023/MaterialFunctions/MF_FN_Hair_2023}
  n4 = GetMaterialAttributes(None=@n3.Result) {attributeGetTypes [#1, #2]}
  n5 = Add(A=@n4.Emissive Color, B=GetMaterialAttributes(None=@n2.Result) {attributeGetTypes [#1]}.Emissive Color)
  n6 = TextureSampleParameter2D() {
   desc="R:Hair mask G; Hairstrands"
   group=FN_Hair
   parameterName=Hair Mask
   samplerSource=SSM_Wrap_WorldGroupSettings
   samplerType=SAMPLERTYPE_Masks
   texture=&T_BlackMask
  }
  n7 = StaticSwitchParameter(
   True=LinearInterpolate(A=@n4.Roughness, B=ScalarParameter() {group=01_SliceVine, parameterName=face pattern roughness}, Alpha=@n6.B)
   False=@n4.Roughness
  ) {group=01_SliceVine, parameterName=Is Cherry Variant, sortPriority=8}
  n8 = BlendMaterialAttributes(
   A=@n2.Result
   B=StaticSwitchParameter(
    True=SetMaterialAttributes!(MaterialAttributes=@n4.MaterialAttributes, Emissive Color=@n5, Roughness=@n7) {attributeSetTypes [#1, #2]}
    False=@n2.Result
   ) {defaultValue=true, group=Skin FX, parameterName=UseFNHair2_0}
   Alpha=@n3.HairMask
  )
  n9 = GetMaterialAttributes(None=@n8) {attributeGetTypes [#3]}
  n10 = SetMaterialAttributes!(
   MaterialAttributes=@n8
   Base Color=@n1.RGB
   Metallic=ScalarParameter() {defaultValue=1, group=01_SliceVine, parameterName=Metallic, sortPriority=11}
   Specular=ScalarParameter() {defaultValue=1, group=01_SliceVine, parameterName=Spec, sortPriority=12}
   Roughness=ScalarParameter() {defaultValue=0.35, group=01_SliceVine, parameterName=Rough, sortPriority=13}
   Shading Model=ShadingModel()
   Normal=MaterialFunctionCall(Normal=@n9.Normal, Flatness=ScalarParameter() {group=01_SliceVine, parameterName=normal flat}) {materialFunction=&FlattenNormal}.Result
  ) {attributeSetTypes [
    #4
    #5
    #6
    #2
    #7
    #3
   ]}
  n11 = Saturate(None=Multiply(A=ComponentMask(None=@n2.Tech Art Mask) {g=true}) {constB=3})
  n12 = BlendMaterialAttributes(
   A=@n10
   B=SetMaterialAttributes!(
    MaterialAttributes=@n8
    Base Color=VectorParameter() {group=01_SliceVine, parameterName=Color Alt, sortPriority=10, defaultValue=<rgba 0 0 0 1>}.RGB
   ) {attributeSetTypes [#4]}
   Alpha=@n11
  )
  n13 = Multiply(A=TextureCoordinate(), B=ScalarParameter() {defaultValue=3, group=01_SliceVine, parameterName=Tiling, sortPriority=9})
  n14 = ComponentMask(None=@n13) {r=true}
  n15 = ScalarParameter() {group=01_SliceVine, parameterName=Offset X, sortPriority=9}
  n16 = ComponentMask(None=@n13) {g=true}
  n17 = ScalarParameter() {group=01_SliceVine, parameterName=Offset Y, sortPriority=9}
  n18 = TextureSampleParameter2D(UVs=AppendVector(A=Add(A=@n14, B=@n15), B=Add(A=@n16, B=@n17))) {
   constCoordinate=1
   group=01_SliceVine
   parameterName=SliceVineDecals
   sortPriority=1
   texture=/BRCosmetics/Characters/Player/Male/Medium/Bodies/M_MED_SliceVine_LoopPlum/Skins/Loop/Textures/T_SliceVine_Decals_Loop
  }
  n19 = Multiply(
   A=ComponentMask(None=@n2.Tech Art Mask) {b=true}
   B=Power(Base=@n18.A, Exp=ScalarParameter() {defaultValue=1, group=01_SliceVine, parameterName=mask pow})
  )
  n20 = TextureSampleParameter2D() {
   constCoordinate=1
   group=01_SliceVine
   parameterName=SliceVineDecals
   sortPriority=1
   texture=&T_SliceVine_Decals_Loop
  }
  n21 = Power(Base=@n20.G, Exp=ScalarParameter() {defaultValue=1, group=01_SliceVine, parameterName=mask pow})
  n22 = Saturate(None=Add(A=@n20.G, B=@n20.G))
  n23 = StaticSwitchParameter(True=Add(A=@n19, B=LinearInterpolate(A=@n21, B=@n22, Alpha=@n11)), False=@n19) {defaultValue=true, group=01_SliceVine, parameterName=UseDecalSheet, sortPriority=0}
  n24 = GetMaterialAttributes(None=@n8) {attributeGetTypes [#4]}
  n25 = BlendMaterialAttributes(
   A=SetMaterialAttributes(MaterialAttributes=@n24.MaterialAttributes, Base Color=LinearInterpolate(A=@n24.Base Color, B=@n20.RGB, Alpha=@n20.RGB)) {attributeSetTypes [#4]}
   B=@n9.MaterialAttributes
   Alpha=ComponentMask(None=@n2.Tech Art Mask) {b=true}
  )
  MaterialFunctionCall(
   Material=StaticSwitchParameter(
    True=BlendMaterialAttributes(A=@n8, B=@n12, Alpha=@n23)
    False=StaticSwitchParameter(True=@n25, False=@n8) {defaultValue=true, group=01_SliceVine, parameterName=UseDecalSheet, sortPriority=0}
   ) {group=01_SliceVine, parameterName=Is Cherry Variant, sortPriority=8}
   M=@n2.M
   Pre Skinned Local Position=@n2.M
   Distance Blend=@n2.M
   Fresnel=@n2.M
  ) {materialFunction=&MF_CharacterParent_2023_PostFX}
 ]
}
